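; PCG debug graph that visualizes Polygon2D, spline and point data from one untyped input, filtered by type.
; Polygon2Ds become points, then closed linear splines, and are spawned as spline components alongside incoming splines.
; Points get noise on their color or, with RandomColorPerData, one color per data set via the OneColorPerDataSet subgraph.
; AbsoluteExtents optionally routes points through an ExtentsSize-driven extents modifier before the Debug node displays them.
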(graph Debug_PrimitiveData
  (params (AbsoluteExtents :type Bool) (ExtentsSize :type Double :default 5e+02) (RandomColorPerData :type Bool))
  (inputs (PrimaryInput :type Any))
  ;[0 0 0 1 32] Debug
  ;;[1 0.1667 0] Display Splines and Polygon2Ds
  (bind FilterDataByType_0 (Filter%20Data%20By%20Type :TargetType (struct :Ids [(struct :Struct /Script/PCG.PCGDataTypeInfoPolygon2D)])))
  (bind ToPoint_0 (To%20Point))
  (bind CreateSpline_0 (Create%20Spline :bClosedLoop true :bLinear true :InterpTypeAttribute None))
  (bind FilterDataByType_3 (Filter%20Data%20By%20Type :TargetType (struct :Ids [(struct :Struct /Script/PCG.PCGDataTypeInfoSpline)])))
  (bind SpawnSplineComponent_1 (Spawn%20Spline%20Component :bOutputSplineComponentReference false))
  ;;[1 0.1667 0] Display Points
  (bind FilterDataByType_4 (Filter%20Data%20By%20Type :TargetType (struct :Ids [(struct :Struct /Script/PCG.PCGDataTypeInfoPoint)])))
  (bind Branch_1 (Branch :bOutputToB-param RandomColorPerData))
  (bind AttributeNoise_0 (Attribute%20Noise :InputSource $color))
  (bind Subgraph_0 (Subgraph :Graph /PCGPrimitives/Primitives/Debug/SubGraphs/OneColorPerDataSet.OneColorPerDataSet))
  (bind Branch_0 (Branch :bOutputToB-param AbsoluteExtents))
  (bind ExtentsModifier_0 (Extents%20Modifier :Extents-param ExtentsSize))
  (bind Debug_0 (Debug))
  (connect Branch_1 Output%20A AttributeNoise_0 In)
  (connect AttributeNoise_0 Out Branch_0 In)
  (connect Subgraph_0 Out Branch_0 In)
  (connect FilterDataByType_4 InsideFilter Branch_1 In)
  (connect ToPoint_0 Out CreateSpline_0 In)
  (connect Branch_0 Output%20A Debug_0 In)
  (connect ExtentsModifier_0 Out Debug_0 In)
  (connect Branch_0 Output%20B ExtentsModifier_0 In)
  (connect DefaultInputNode PrimaryInput FilterDataByType_0 In)
  (connect DefaultInputNode PrimaryInput FilterDataByType_3 In)
  (connect DefaultInputNode PrimaryInput FilterDataByType_4 In)
  (connect CreateSpline_0 Out SpawnSplineComponent_1 In)
  (connect FilterDataByType_3 InsideFilter SpawnSplineComponent_1 In)
  (connect Branch_1 Output%20B Subgraph_0 In)
  (connect FilterDataByType_0 InsideFilter ToPoint_0 In))
